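# Additive unlit lens flare material for particle sprites and Niagara meshes, apparently for a neon rocket effect.
# Opacity multiplies flare texture by particle alpha and time-of-day values, with blendable depth fade (skipped on mobile) and near-camera fade.
# Emissive samples one parameter-picked texel from a neon glow texture, scales it by intensity and a dynamic parameter, or blends toward particle color, then applies HDR-from-LDR and shared FX adjustments.
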Material /Game/Effects/Materials/Flares/M_NeonRocket_LensFlare
graph {
 properties {
  bAutomaticallySetUsageInEditor=false
  bUsedWithNiagaraMeshParticles=true
  bUsedWithNiagaraSprites=true
  bUsedWithParticleSprites=true
  bUsedWithStaticMesh=false
  blendMode=BLEND_Additive
  pixelDepthOffsetMode=PDOM_Legacy
  shadingModel=MSM_Unlit
  usedShadingModels=Unlit
 }
 nodes [
  n1 = TextureSampleParameter2D() {
   group=Opacity
   parameterName=LFTexture
   samplerType=SAMPLERTYPE_Grayscale
   sortPriority=0
   texture=/Game/Effects/Fort_Effects/Textures/T_EFX_CameraFlare_01
  }
  n2 = Multiply(A=ParticleColor().A, B=@n1.R)
  n3 = DepthFade(Opacity=@n2, FadeDistance=ScalarParameter() {defaultValue=30, group=Opacity|DepthFade, parameterName=FadeDistance, sortPriority=2})
  n4 = ScalarParameter() {group=Opacity|DepthFade, parameterName=UseDepthFade, sortPriority=1}
  n5 = Multiply(
   A=MaterialFunctionCall(
    Color_Morning=ScalarParameter() {defaultValue=1, group=Opacity, parameterName=OpacityMorning, sortPriority=1}
    Color_Day=ScalarParameter() {defaultValue=1, group=Opacity, parameterName=OpacityDay, sortPriority=2}
    Color_Evening=ScalarParameter() {defaultValue=1, group=Opacity, parameterName=OpacityEvening, sortPriority=3}
    Color_Night=ScalarParameter() {defaultValue=1, group=Opacity, parameterName=OpacityNight, sortPriority=4}
   ) {materialFunction=&Color_Time_of_Day}.Result
   B=ShadingPathSwitch(Default=LinearInterpolate(A=@n2, B=@n3, Alpha=@n4), Mobile=@n2) {desc=Migrated from FeatureLevelSwitch}
  )
  n6 = ScalarParameter() {defaultValue=256, group=Opacity|NearCameraFade, parameterName=NearCameraFadeDistance, sortPriority=1}
  LinearInterpolate(
   A=@n5
   B=Multiply(A=@n5, B=MaterialFunctionCall(Distance=@n6) {materialFunction=&Near_Camera_Fade}.Result)
   Alpha=ScalarParameter() {group=Opacity|NearCameraFade, parameterName=Use NearCameraFade, sortPriority=0}
  )
  n7 = TextureSampleParameter2D(UVs=AppendVector(A=Constant() {r=1}, B=ScalarParameter() {defaultValue=0.9, group=Emissive, parameterName=Texture_Color_Selection, sortPriority=1})) {group=Emissive, parameterName=Color Selection Texture, sortPriority=0, texture=/Game/Weapons/FORT_Rifles/Textures/Neon_Glow_GlowParts_E}
  n8 = ScalarParameter() {defaultValue=5, group=Emissive, parameterName=EmissiveIntensity, sortPriority=2}
  n9 = DynamicParameter() {paramNames [Overall_Intensity, Param2, Param3, Param4]}
  n10 = Clamp(None=Multiply(A=@n7.RGB, B=Multiply(A=@n8, B=@n9.Overall_Intensity)) {constB=3})
  n11 = ParticleColor()
  n12 = Clamp(None=@n11.RGB)
  n13 = ScalarParameter() {group=Emissive, parameterName=Use Particle Color, sortPriority=4}
  MaterialFunctionCall(Emissive=MaterialFunctionCall(LDR Input=LinearInterpolate(A=@n10, B=@n12, Alpha=@n13), HDR Tint=LinearInterpolate(A=@n10, B=@n11.RGB, Alpha=@n13)) {materialFunction=&DeriveHDRfromLDR}.RemappedRange) {materialFunction=&MF_FX_EmissiveAdjustments}
 ]
}
