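# Unlit additive flare material for sprite and mesh particles, with a hot center, soft edge and flicker.
# Powered radial masks and a flare texture shape the emissive, tinted by particle color or rainbow, with optional hue jitter and radial gradient.
# Opacity combines depth, angle and camera fades with multiplied-sine flicker, driven by dynamic parameters or per-particle random for GPU particles.
# Optional world position offset scales flares with camera distance.
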Material /Game/Effects/Materials/Flares/M_Particle_Flare_HotCenter_Flicker
graph {
 properties {
  bAllowNegativeEmissiveColor=true
  bAutomaticallySetUsageInEditor=false
  bUsedWithNiagaraMeshParticles=true
  bUsedWithNiagaraSprites=true
  bUsedWithParticleSprites=true
  bUsedWithStaticMesh=false
  blendMode=BLEND_Additive
  pixelDepthOffsetMode=PDOM_Legacy
  shadingModel=MSM_Unlit
  usedShadingModels=Unlit
 }
 nodes [
  n1 = Constant() {r=0.5}
  n2 = SphereMask(A=TextureCoordinate(), B=@n1, Radius=@n1) {hardnessPercent=0}
  n3 = Power(Base=@n2, Exp=ScalarParameter() {defaultValue=25, group=Flare Emissive Shaping, parameterName=CentertPower, sortPriority=0}) {constExponent=25}
  n4 = ScalarParameter() {defaultValue=2, group=Flare Emissive Shaping, parameterName=CenterBrightness, sortPriority=1}
  n5 = NamedRerouteUsage()
  n6 = ScalarParameter() {defaultValue=0.3, group=Flare Color, parameterName=CenterDesaturation, sortPriority=1}
  n7 = Multiply(
   A=Power(Base=@n2, Exp=ScalarParameter() {defaultValue=4, group=Flare Emissive Shaping, parameterName=EdgePower, sortPriority=2}) {constExponent=4}
   B=ScalarParameter() {defaultValue=0.15, group=Flare Emissive Shaping, parameterName=EdgeBrightness, sortPriority=3}
  ) {constB=0.15}
  n8 = NamedRerouteUsage()
  n9 = Add(A=Multiply(A=Multiply(A=@n3, B=@n4) {constB=2}, B=Desaturation(None=@n5, Fraction=@n6)), B=Multiply(A=@n5, B=Add(A=@n7, B=@n8)))
  n10 = LinearInterpolate(
   A=@n9
   B=MaterialFunctionCall(Hue Shift Percentage=DynamicParameter() {paramNames [FlickerFade,FlickerSpeed,ColorJitter,Param4]}.ColorJitter, Texture=@n9) {materialFunction=&HueShift}.Result
   Alpha=ScalarParameter() {group=Flare Color, parameterName=Use Color Jitter, sortPriority=3}
  )
  n11 = NamedRerouteUsage()
  n12 = MaterialFunctionCall(Emissive=LinearInterpolate(
    A=@n10
    B=Multiply(A=@n11, B=@n10)
    Alpha=ScalarParameter() {group=RadialGradientColor, parameterName=Use Radial Gradient Color, sortPriority=0}
   )) {materialFunction=&MF_FX_EmissiveAdjustments}
  LinearInterpolate(A=@n12.Result, B=Max(A=@n12.Result) {constB=0}, Alpha=ScalarParameter() {parameterName=ClampBottomForEmissive})
  n13 = ScalarParameter() {defaultValue=10, group=WPO_Distance Fade, parameterName=MaxScaleIncrease, sortPriority=1}
  n14 = ObjectPositionWS()
  n15 = Multiply(A=Subtract(A=@n14, B=WorldPosition().XYZ)) {constB=-1}
  n16 = CameraPositionWS()
  n17 = Divide(
   A=Subtract(A=Distance(A=@n14, B=@n16), B=ScalarParameter() {defaultValue=2000, group=WPO_Distance Fade, parameterName=DistanceOffset, sortPriority=3})
   B=ScalarParameter() {defaultValue=30000, group=WPO_Distance Fade, parameterName=ScaleUpDistance, sortPriority=2}
  )
  LinearInterpolate(
   B=LinearInterpolate(B=Multiply(A=@n13, B=@n15) {constB=-1}, Alpha=Clamp(None=@n17)) {constB=0}
   Alpha=ScalarParameter() {group=WPO_Distance Fade, parameterName=Use WPO_DistanceFade, sortPriority=0}
  )
  n18 = ParticleColor()
  n19 = DynamicParameter() {desc=Index 1 Params for Niagara, parameterIndex=1, paramNames [DepthFadeDist, Param2, Param3, Param4]}
  n20 = ScalarParameter() {defaultValue=30, group=Opacity, parameterName=DepthFade, sortPriority=12}
  n21 = LinearInterpolate(
   A=If(A=@n19.DepthFadeDist, B=Constant() {r=1}, "A > B"=@n19.DepthFadeDist, "A == B"=@n20, "A < B"=@n20)
   B=@n20
   Alpha=ScalarParameter() {group=Opacity, parameterName=SimpleDepthFade, sortPriority=11}
  )
  n22 = Constant() {r=1}
  n23 = MaterialFunctionCall("FadeSensitivity (optional)"=ScalarParameter() {defaultValue=1, group=Opacity, parameterName=AngleFade, sortPriority=10}) {materialFunction=&FadeBasedOnViewAngle}
  n24 = ScalarParameter() {group=Opacity, parameterName=UseAngleFade, sortPriority=9}
  n25 = Constant() {r=1}
  n26 = ParticleRandom()
  n27 = ScalarParameter() {group=Opacity, parameterName=Random Flicker Offset}
  n28 = ParticleRandom()
  n29 = Multiply(
   A=Add(A=Time(), B=LinearInterpolate(B=@n26, Alpha=@n27))
   B=LinearInterpolate(
    A=DynamicParameter() {paramNames [FlickerFade,FlickerSpeed,ColorJitter,Param4]}.FlickerSpeed
    B=LinearInterpolate(Alpha=@n28) {constA=1.2, constB=0.8}
    Alpha=ScalarParameter() {group=Opacity, parameterName=Use GPU Particles, sortPriority=14}
   )
  )
  n30 = Sine(None=@n29) {period=0.075}
  n31 = Multiply(A=Multiply(A=Sine(None=@n29) {period=0.155}, B=Sine(None=@n29) {period=0.2}), B=Sine(None=@n29) {period=0.43})
  n32 = DynamicParameter() {paramNames [FlickerFade,FlickerSpeed,ColorJitter,Param4]}
  n33 = ScalarParameter() {defaultValue=0.4, desc="0 = full flicker amount", group=Opacity, parameterName=GPUFlickerFadeMin, sortPriority=15}
  n34 = ScalarParameter() {defaultValue=1, desc="1 = no flicker", group=Opacity, parameterName=GPUFlickerFadeMax, sortPriority=16}
  n35 = ParticleRandom()
  n36 = LinearInterpolate(
   A=Clamp(None=Multiply(A=@n30, B=@n31)) {maxDefault=3}
   Alpha=LinearInterpolate(
    A=Saturate(None=@n32.FlickerFade)
    B=LinearInterpolate(A=@n33, B=@n34, Alpha=@n35) {constA=0.5}
    Alpha=ScalarParameter() {group=Opacity, parameterName=Use GPU Particles, sortPriority=14}
   )
  ) {constA=1}
  n37 = ScalarParameter() {defaultValue=1, group=Opacity, parameterName=Use Flicker, sortPriority=13}
  n38 = Multiply(
   A=ShadingPathSwitch(Default=DepthFade(Opacity=@n18.A, FadeDistance=@n21), Mobile=@n18.A)
   B=Multiply(A=LinearInterpolate(A=@n22, B=@n23.Result, Alpha=@n24), B=LinearInterpolate(A=@n25, B=@n36, Alpha=@n37))
  )
  n39 = MaterialFunctionCall(Distance=ScalarParameter() {group=Opacity, parameterName=NearCameraFadeDistance, sortPriority=5}) {materialFunction=&Near_Camera_Fade}
  n40 = MaterialFunctionCall(
   Fade Length=ScalarParameter() {group=Opacity, parameterName="Cam Fade Length (S)", sortPriority=7}
   Fade Offset=ScalarParameter() {group=Opacity, parameterName="Cam Fade Offset (S)", sortPriority=8}
  ) {materialFunction=&CameraDepthFade}
  n41 = ScalarParameter() {group=Opacity, parameterName=Use Better Camera Fade, sortPriority=6}
  n42 = LinearInterpolate(
   A=@n38
   B=Multiply(A=@n38, B=LinearInterpolate(A=@n39.Result, B=@n40.Result, Alpha=@n41))
   Alpha=ScalarParameter() {group=Opacity, parameterName=Use CameraNearFade, sortPriority=4}
  )
  n43 = LinearInterpolate(A=@n42, B=Multiply(A=@n42, B=NamedRerouteUsage()), Alpha=ScalarParameter() {group=Opacity, parameterName=Use CameraFarFade, sortPriority=1})
  LinearInterpolate(
   A=@n43
   B=Multiply(A=@n43, B=NamedRerouteUsage())
   Alpha=ScalarParameter() {group=Opacity, parameterName=Use FlareMaskOpacity, sortPriority=0}
  )
  n44 = TextureCoordinate()
  n45 = MaterialFunctionCall(XY=@n44) {materialFunction=&Swizzle}
  n46 = ScalarParameter() {group=Flare Texture, parameterName=UseSwizzle, sortPriority=2}
  n47 = VectorParameter() {group=Flare Texture, parameterName=FlareOffset, sortPriority=3, defaultValue=<rgba 0 0 0 1>}
  n48 = TextureSampleParameter2D(UVs=Add(A=LinearInterpolate(A=@n44, B=@n45.YX, Alpha=@n46), B=ComponentMask(None=@n47.RGB) {g=true, r=true})) {group=Flare Texture, parameterName=FlareMask, sortPriority=0, texture=/Game/Effects/Fort_Effects/Textures/Flares/T_Flare_Spikeball}
  NamedRerouteDeclaration(None=ChannelMaskParameter(None=@n48.RGBA) {desc="Default = Red", group=Flare Texture, parameterName=FlareMask_OpacityChannel, sortPriority=1}.R) {name=FlareMask_Opacity, nodeColor=<rgba 1 0 0.5859 1>}
  n49 = VectorParameter() {group=RadialGradientColor, parameterName=EdgeColor, sortPriority=1, defaultValue=<rgba 1 0 0.6939 1>}
  n50 = VectorParameter() {group=RadialGradientColor, parameterName=CenterColor, sortPriority=2, defaultValue=<rgba 0 0.8019 1 1>}
  n51 = VectorParameter() {
   desc=(R=Radius, G=Hardness, B=Contrast (Power))
   group=RadialGradientColor
   parameterName=RadialGradientColorSetting
   sortPriority=3
   channelNames {a="", b=P, g=H, r=R}
   defaultValue=<rgba 1 0 1 1>
  }
  n52 = SphereMask(A=TextureCoordinate(), B=Constant() {r=0.5}, Radius=@n51.R, Hardness=@n51.R) {attenuationRadius=1, hardnessPercent=0}
  NamedRerouteDeclaration(None=LinearInterpolate(A=@n49.RGB, B=@n50.RGB, Alpha=Power(Base=@n52, Exp=@n51.P))) {name=RadialGradientColor, nodeColor=<rgba 0.4062 1 0 1>}
  NamedRerouteDeclaration(None=@n48.RGB) {name=FlareMask_RGB, nodeColor=<rgba 1 0.8203 0 1>}
  n53 = Length(None=Subtract(A=CameraPositionWS(), B=WorldPosition().XYZ)) {is=class MaterialExpressionLength}
  n54 = ScalarParameter() {defaultValue=1000, group=Opacity, parameterName=Camera Fade Far Start, sliderMax=50000, sortPriority=2}
  n55 = ScalarParameter() {defaultValue=3000, group=Opacity, parameterName=Camera Fade Far End, sliderMax=70000, sortPriority=3}
  n56 = Constant() {r=1}
  n57 = Constant()
  NamedRerouteDeclaration(None=Saturate(None=MaterialFunctionCall(Input=@n53, Input Low=@n54, Input High=@n55, Target Low=@n56, Target High=@n57) {materialFunction=&RemapValueRange}.Result)) {name=Camera Fade Far, nodeColor=<rgba 1 0.0469 0 1>}
  n58 = ParticleColor()
  n59 = ScalarParameter() {desc=Desaturation value for Particle Color, group=Flare Color, parameterName=inputDesaturation, sortPriority=0}
  n60 = ComponentMask(None=TextureCoordinate() {uTiling=0, vTiling=0}) {g=true}
  n61 = Multiply(
   A=TextureSample(UVs=Panner(Coordinate=@n60) {speedX=1.6667}) {texture=/Game/Global/Textures/Default/Masks_and_Gradients/T_Grad_Rainbow}.RGB
   B=ParticleColor().RGB
  )
  n62 = Saturate(None=@n61)
  NamedRerouteDeclaration(None=LinearInterpolate(
    A=Desaturation(None=@n58.RGB, Fraction=@n59)
    B=MaterialFunctionCall(LDR Input=@n62, HDR Tint=@n61) {materialFunction=&DeriveHDRfromLDR}.RemappedRange
    Alpha=ScalarParameter() {group=Flare Color, parameterName=Use Rainbow Color, sortPriority=2}
   )) {name=Flare Color, nodeColor=<rgba 0.5 0 1 1>}
 ]
}
